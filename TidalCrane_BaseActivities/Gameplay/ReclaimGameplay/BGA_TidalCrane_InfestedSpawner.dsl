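; Infested spawner activity actor that runs an AI swarm encounter and rewards players for clearing it.
; When the playlist is ready, the dedicated server enables its living world point provider, starts the encounter and binds spawn and death events.
; Every tenth death bursts an egg sac, with OnRep_EggSacCounter playing Niagara effects on clients.
; After four sacs, it pauses rather than ends the encounter and tosses out Loot_Golden_Llama loot.
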(event EventBeginPlay
  (bind _livingworldvolume_spawning (Variables|Default|GetLivingWorldVolume_Spawning))
  (bind _childactor (Variables|/Script/Engine.ChildActorComponent|GetChildActor _livingworldvolume_spawning))
  (Networking|CanExecuteCosmeticEvents
    (:True
      (Components|DestroyComponent (Variables|Default|GetSpawnCollision)))
    (:False
      (Variables|InternalData|SetEncounterInProgress false)
      (Class|FortAsyncActionCurrentPlaylistReady|PlaylistDataReadyAsync
        (:OnReady
          (Utilities|IsValid _childactor
            (:Is%20Valid
              (bind _asfort_athena_living_world_volume (Utilities|Casting|CastToFortAthenaLivingWorldVolume _childactor))
              (Class|FortAthenaLivingWorldPointProviderInterface|EnablePointProvider_DEPRECATED _asfort_athena_living_world_volume)
              (CallFunction|SpawnAI))))))))

(event Custom|SpawnAI
  ; Try Spawn AI: ToDo - Track Velocity from Pawns Inside Volume, not just when entering
  (Variables|InternalData|SetEncounterInProgress true)
  (CallFunction|ProviderRegistered 0))

(event Custom|ProviderRegistered
  (params (PointProviderVolume :type Fort%20Athena%20Living%20World%20Volume%20Object%20Reference))
  ; Start Encounter for AI
  (Utilities|FlowControl|Delay 0.2)
  (bind (_isvalid _returnvalue) (Default|StartActivityEncounterStaticPointProvider (Game|GetGameState) self (Utilities|ResolveSoftReference (Variables|Default|GetLagerEncounter)) (Variables|Default|GetStaticPointProvider)))
  (if _isvalid
    (bind _output_get (Variables|InternalData|SetEncounterRef _returnvalue))
    (Class|FortAthenaLivingWorldManager|RequestEventGeneration (Class|FortAthenaLivingWorldManager|GetLivingWorldManager))
    (EventDispatchers|BindEventtoOnActorDied _output_get (EventDispatchers|CreateEvent :!event Custom|SwarmAIDied))
    (EventDispatchers|BindEventtoOnActorSpawned _output_get (EventDispatchers|CreateEvent :!event Custom|OnAISpawned))
    (EventDispatchers|BindEventtoOnLastActorDied _output_get (EventDispatchers|CreateEvent :!event Custom|InitialEncounterLastAIDied))
    (else
      (Fortnite|FortShippingLog self "Tried to Spawn Encounter for Mound, but Encounter failed to start." true))))

(event Custom|OnAISpawned
  (params (EncounterInstance :type Fort%20Athena%20Living%20World%20Encounter%20Instance%20Object%20Reference) (SpawnedActor :type Actor%20Object%20Reference) (EventData :type Fort%20Athena%20Living%20World%20Event%20Data%20Object%20Reference))
  (Default|CallSpawnAIEvent SpawnedActor))

(event Custom|SwarmAIDied
  (params (EncounterInstance :type Fort%20Athena%20Living%20World%20Encounter%20Instance%20Object%20Reference) (DiedActor :type Actor%20Object%20Reference))
  ; Track Swarm Deaths and Complete Encounter: ToDo - Encounter is just PAUSED and not Ended. Make sure we end it when last creature is killed
  (bind _result (Math|Integer|IncrementInt (Variables|InternalData|GetDeathCount))
    (:%20%20
      (if (and (== (Math|Integer|%25%28Integer%29 _result 10) 0) (not (Variables|InternalData|GetObjectiveCompleted)))
        (CallFunction|SetEggDatatoExplode (Utilities|Select :Index (Variables|InternalData|GetEggSacCounter)))
        (Utilities|FlowControl|Delay 0.2)
        (Variables|InternalData|SetEggSacCounter (+ (Variables|InternalData|GetEggSacCounter) 1))
        (Rendering|SetVisibility (Variables|InternalData|GetEggToExplode) false false)
        (if (>= (Variables|InternalData|GetEggSacCounter) 4)
          (Class|FortAthenaLivingWorldEncounterInstance|PauseEncounter (Variables|InternalData|GetEncounterRef))
          (Inventory|SpawnPickupInWorldWithLootTier "Loot_Golden_Llama" (+ (+ (Transformation|GetActorLocation self) "0.000000,0.000000,650.000000") (* (Transformation|GetActorForwardVector self) 1.5e+02)) 0 0 true true "0.000000,0.000000,300.000000" "Tossed" "ItemSpawner")
          (Variables|InternalData|SetObjectiveCompleted true))))))

(event Custom|InitialEncounterLastAIDied
  (params (EncounterInstance :type Fort%20Athena%20Living%20World%20Encounter%20Instance%20Object%20Reference))
  ; End the Initial Encounter - If there was one
  (Class|BPFLLagerCommunFunctions|EndActivityEncounter (Game|GetGameState) (Variables|InternalData|GetInitialEncounterRef)))

(fn OnRep_EggSacCounter
  (Networking|CanExecuteCosmeticEvents
    (:True
      (Niagara|SpawnSystematLocation :SystemTemplate "/ProtoSwarm/Assets/FX/NS_AvacadoEater_Bird_Destroyed.NS_AvacadoEater_Bird_Destroyed" :Location (Variables|InternalData|GetEggSacExplosionLocation) :Scale "1.000000,1.000000,1.000000" :bAutoDestroy true :bAutoActivate true :PoolingMethod "None" :bPreCullCheck false)
      (Rendering|SetVisibility (Variables|InternalData|GetEggToExplode) false false))))

(fn OnRep_EncounterInProgress)

(fn SetEggDataToExplode
  (params (EggToExplode :type Static%20Mesh%20Component%20Object%20Reference))
  (bind _output_get (Variables|InternalData|SetEggToExplode EggToExplode))
  (Variables|InternalData|SetEggSacExplosionLocation (Transformation|GetWorldLocation _output_get)))

(fn ConstructionScript)
