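; Spawn manager for the Tidal Crane base Reclaim activity, won by destroying all bug spawners and freeing every cocooned NPC in the event area.
; On authority it binds spawners and registry-found cocoons, tracking progress in EventData, which OnRep broadcasts through GameplayEventRouter.
; Winning optionally kills bugs in the box via hotfix, signals NPCs, sends quest credit and spawns chests.
; Storm consumption ends the encounter as a failure.
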(event EventBeginPlay
  (bind _bugspawners_continuous (Variables|Default|GetBugSpawners_Continuous))
  ; BeginPlay
  (|CallParentFunction)
  (Networking|SwitchHasAuthority
    (:Authority
      (CallFunction|RegisterCocoons)
      (bind _output_get (Variables|Default|SetNumber%20of%20ActiveSpawners (Utilities|Array|Length (Variables|Default|GetBugSpawners_Continuous))))
      (bind _structout (Utilities|Struct|SetmembersinTidalCraneBaseReclamationData (Variables|Default|GetEventData) self (Utilities|Array|Length (Variables|Default|GetCocoons)) _output_get))
      (Variables|Default|SetEventData _structout)
      (Game|AssignOnDestroyed _bugspawners_continuous (delegate Custom|OnDestroyed_Event))
      (Default|BindEventtoOnSpawnedAnActor _bugspawners_continuous (EventDispatchers|CreateEvent :!event Custom|AISpawned)))))

(event Custom|OnDestroyed_Event
  (params (DestroyedActor :type Actor%20Object%20Reference))
  (bind _numdestroyedspawners (Variables|Default|GetNumDestroyedSpawners))
  ; BeginPlay
  (Development|PrintString "Popped a spawner")
  (Math|Integer|IncrementInt _numdestroyedspawners
    (:%20%20
      (bind _structout (Utilities|Struct|SetmembersinTidalCraneBaseReclamationData (Variables|Default|GetEventData) _numdestroyedspawners))
      (Variables|Default|SetEventData _structout)
      (Variables|Default|SetNumber%20of%20ActiveSpawners (- (Variables|Default|GetNumber%20of%20ActiveSpawners) 1))
      (CallFunction|CheckEndConditions))))

(event Custom|AISpawned
  (params (Spawned%20Actor :type Actor%20Object%20Reference))
  ;;[0.01931 0.15 0.09414 0.5] AI Spawned
  ; On spawned for tracking the bugs
  (Utilities|Array|Add (Variables|Encounter|GetActiveMinions) Spawned%20Actor) ; Keep active minions \ntracked
  (CallFunction|UpdateLootTagforAI Spawned%20Actor)
  (Game|BindEventtoOnDestroyed Spawned%20Actor (delegate Custom|Bug%20Died)))

(event Custom|OnCocoonAdded
  (params (GameplayTag :type Gameplay%20Tag%20Structure%20%28by%20ref%29) (RegisteredActor :type Actor%20Object%20Reference))
  (bind (_spawnmanager_2_73860b3c407b1bc6e2a4bbbc629f37fa _eventstate_5_781d3b7c4b9a8efdcfc0f5832d1ce7c1 _freednpcs_8_ba4606824ff6aa73adbf20ac85ab4e02 _maxnpcstofree_17_feb9e4fd4cd339ff26e5e7af3e52f04c) (Utilities|Struct|BreakTidalCraneBaseReclamationData (Variables|Default|GetEventData)))
  ; Handle Rescued NPC Logic
  (Networking|SwitchHasAuthority
    (:Authority
      (bind _asbga_loot_sac_npc (Utilities|Casting|CastToBGA_LootSac_NPC RegisteredActor))
      (bind _binarea (CallFunction|IsActorinEventArea _asbga_loot_sac_npc))
      (if _binarea
        (Utilities|Array|Add (Variables|Default|GetCocoons) _asbga_loot_sac_npc)
        (NPC|BindEventtoRescuedNPC _asbga_loot_sac_npc (delegate Custom|OnNPCRescued))
        (NPC|BindEventtoOnNPCSpawned _asbga_loot_sac_npc (delegate Custom|OnNPCSpawned))
        (bind _structout (Utilities|Struct|SetmembersinTidalCraneBaseReclamationData (Variables|Default|GetEventData) (+ _maxnpcstofree_17_feb9e4fd4cd339ff26e5e7af3e52f04c 1)))
        (Variables|Default|SetEventData _structout)))))

(event Custom|Bug%20Died
  (params (DestroyedActor :type Actor%20Object%20Reference))
  ;;[0.01931 0.15 0.09414 0.5] AI Spawned
  ; On spawned for tracking the bugs
  (Utilities|Array|RemoveItem (Variables|Encounter|GetActiveMinions) DestroyedActor))

(event Custom|OnNPCRescued
  (bind (_spawnmanager_2_73860b3c407b1bc6e2a4bbbc629f37fa _eventstate_5_781d3b7c4b9a8efdcfc0f5832d1ce7c1 _freednpcs_8_ba4606824ff6aa73adbf20ac85ab4e02) (Utilities|Struct|BreakTidalCraneBaseReclamationData (Variables|Default|GetEventData)))
  ; Handle Rescued NPC Logic
  ;;[0.15 0 0.1486 0.5] Track Number of NPCs Rescued and Check Victory Condition
  (bind _structout (Utilities|Struct|SetmembersinTidalCraneBaseReclamationData (Variables|Default|GetEventData) (+ _freednpcs_8_ba4606824ff6aa73adbf20ac85ab4e02 1)))
  (Variables|Default|SetEventData _structout)
  (CallFunction|CheckEndConditions))

(event Custom|OnNPCSpawned
  (params (SpawnedActor :type Actor%20Object%20Reference))
  ; Handle Rescued NPC Logic
  ;;[0.15 0.01827 0.1326 0.5] Set NPC Leash for event
  (bind _asfort_player_pawn (Utilities|Casting|CastToFortPlayerPawn SpawnedActor))
  (Utilities|Array|Add (Variables|Default|GetNPCs) _asfort_player_pawn)
  (CallFunction|ApplyorRemoveEffectsonNPC false _asfort_player_pawn)
  (Utilities|FlowControl|Delay 0.2)
  (CallFunction|SendGameplayEventtoNPC "(TagName=\"Event.NPC.TidalCrane.BaseReclaim.Rescue\")" _asfort_player_pawn))

(event Custom|CheckEndConditions
  (bind (_spawnmanager_2_73860b3c407b1bc6e2a4bbbc629f37fa _eventstate_5_781d3b7c4b9a8efdcfc0f5832d1ce7c1 _freednpcs_8_ba4606824ff6aa73adbf20ac85ab4e02 _maxnpcstofree_17_feb9e4fd4cd339ff26e5e7af3e52f04c _bugskilled_10_fcec42d04ca2b85d8e9b6eb4ce1804da _spawnersdestroyed_12_94bdde484207b110c717a5b7340381ff _maxspawners_14_c069eae442d53210b1269db2a5ba1ef5) (Utilities|Struct|BreakTidalCraneBaseReclamationData (Variables|Default|GetEventData)))
  ; WinCon Check + Rewards
  (Networking|SwitchHasAuthority
    (:Authority
      (if (Math|Boolean|ANDBoolean (not (or (Utilities|Enum|Equal(Enum) _eventstate_5_781d3b7c4b9a8efdcfc0f5832d1ce7c1 "NewEnumerator5") (Utilities|Enum|Equal(Enum) _eventstate_5_781d3b7c4b9a8efdcfc0f5832d1ce7c1 "NewEnumerator6"))) (== _freednpcs_8_ba4606824ff6aa73adbf20ac85ab4e02 _maxnpcstofree_17_feb9e4fd4cd339ff26e5e7af3e52f04c) (== _spawnersdestroyed_12_94bdde484207b110c717a5b7340381ff _maxspawners_14_c069eae442d53210b1269db2a5ba1ef5))
        (Utilities|FlowControl|Delay 2.5)
        (CallFunction|DestroySpawnedBugs)
        (for _array_element (Variables|Default|GetNPCs)
          (Utilities|IsValid _array_element
            (:Is%20Valid
              (CallFunction|SendGameplayEventtoNPC "(TagName=\"Event.NPC.TidalCrane.BaseReclaim.Complete\")" _array_element))))
        (Default|SendCompletionQuestEvent "(GameplayTags=((TagName=\"Athena.Quests.BaseActivity.Reclaim\")))")
        (CallFunction|SpawnChestRewards)
        (CallFunction|EndEncounter true)))))

(event EventEndPlay
  (params (EndPlayReason :type EEndPlayReason%20Enum :default "Destroyed"))
  (bind _cachedcocoonregistrysubsystem (Variables|Default|GetCachedCocoonRegistrySubsystem))
  ; EndPlay
  (Utilities|IsValid _cachedcocoonregistrysubsystem
    (:Is%20Valid
      (EventDispatchers|UnbindEventfromOnActorUnregistered _cachedcocoonregistrysubsystem (EventDispatchers|CreateEvent :!event Custom|OnCocoonAdded)))))

(fn DestroySpawnedBugs
  (Networking|SwitchHasAuthority
    (:Authority
      (if (ScalableFloat|GetValueasBool (Variables|Hotfixes|GetHFRow%20DestroySpawnedBugs))
        (bind _outactors (Collision|ComponentOverlapActors (Variables|Default|GetBox) (Transformation|GetWorldTransform (Variables|Default|GetBox)) (Utilities|Array|MakeArray "ObjectTypeQuery7") "/Script/FortniteGame.FortPawn"))
        (for _array_element _outactors
          (bind _asfort_pawn (Utilities|Casting|CastToFortPawn _array_element))
          (if (GameplayTags|HasAnyMatchingGameplayTags (Utilities|Casting|ToGameplayTagAssetInterface _asfort_pawn) (GameplayTags|MakeGameplayTagContainerfromArray (Variables|Default|GetGameplay%20Tags)))
            (FortPawn|ForceKill :self _asfort_pawn)))))))

(fn EndEncounter
  (params (Success :type Boolean))
  (bind _returnvalue (select Success "NewEnumerator5" "NewEnumerator6"))
  (Networking|SwitchHasAuthority
    (:Authority
      (bind _structout (Utilities|Struct|SetmembersinTidalCraneBaseReclamationData (Variables|Default|GetEventData) _returnvalue))
      (Variables|Default|SetEventData _structout)
      (Class|BTidalCraneBaseLocator|SetBaseMinimapIcon (Variables|Default|GetBaseLocator) _returnvalue)
      (for _array_element (Variables|Default|GetNPCs)
        (CallFunction|ApplyorRemoveEffectsonNPC true _array_element)))))

(fn GetLeashLocation
  (bind _leashlocationactor (Variables|Encounter|GetLeashLocationActor)
    (:then
      (return (Transformation|GetActorLocation _leashlocationactor)))
    (:else
      (return (Transformation|GetActorLocation)))))

(fn OnConsumedByStorm
  (|CallParentFunction)
  (CallFunction|EndEncounter false))

(fn OnPlayerPresenceChanged
  (|CallParentFunction)
  (for _array_element (Variables|Encounter|GetActiveMinions)
    (CallFunction|UpdateLootTagforAI _array_element)))

(fn OnRep_EventData
  (bind _returnvalue (GameplayEventRouter|GetEventRouter "/Script/FortniteGame.FortPlayspace" self))
  (GameplayEventRouter|BroadcastEvent _returnvalue (Variables|Default|GetEventData) 0))

(fn RegisterCocoons
  (bind _returnvalue (WorldSubsystems|GetFortGlobalActorRegistrySubsystem))
  (bind _outactors (Class|FortGlobalActorRegistrySubsystem|GetRegisteredActors _returnvalue (Variables|Settings|GetCocoonGameplayTag)))
  (for _array_element _outactors
    (bind _asbga_loot_sac_npc (Utilities|Casting|CastToBGA_LootSac_NPC _array_element))
    (CallFunction|OnCocoonAdded (Variables|Settings|GetCocoonGameplayTag) _asbga_loot_sac_npc))
  (bind _returnvalue_2 (Class|FortGlobalActorRegistrySubsystem|GetActorRegistryDataforGameplayTag _returnvalue (Variables|Settings|GetCocoonGameplayTag)))
  (Utilities|IsValid _returnvalue_2
    (:Is%20Valid
      (bind _output_get (Variables|Default|SetCachedCocoonRegistrySubsystem _returnvalue_2))
      (EventDispatchers|BindEventtoOnActorRegistered _output_get (EventDispatchers|CreateEvent :!event OnCocoonAdded)))))

(fn SendCompletionQuestEvent
  (params (OptionalSubjectInfo :type Gameplay%20Tag%20Container%20Structure))
  (|CallParentFunction OptionalSubjectInfo))

(fn SendGameplayEventToNPC
  (params (EventTag :type Gameplay%20Tag%20Structure) (NPC :type Fort%20Player%20Pawn%20Object%20Reference))
  ;[0 0 0] Sends a gameplay event to NPCs to tell them to trigger certain barks, behaviors, etc.
  (Utilities|IsValid NPC
    (:Is%20Valid
      (Ability|SendGameplayEventtoActor NPC EventTag))))

(fn UpdateAIActorLeash
  (params (AIActor :type Actor%20Object%20Reference))
  (bind _row_leashouterradius (Variables|Encounter|GetRow_LeashOuterRadius))
  (bind _returnvalue (ScalableFloat|GetValueatLevel _row_leashouterradius))
  (bind _row_leashinnerradius (Variables|Encounter|GetRow_LeashInnerRadius))
  (bind _returnvalue_2 (ScalableFloat|GetValueatLevel _row_leashinnerradius))
  (Utilities|IsValid AIActor
    (:Is%20Valid
      (if (and (> _returnvalue_2 0) (> _returnvalue 0))
        (bind _asfort_athena_aibot_controller (Utilities|Casting|CastToFortAthenaAIBotController (Fortnite|GetControllerfromActor AIActor)))
        (Class|BPFLAICommonFunctions|SetRadialLeash _asfort_athena_aibot_controller (Default|GetLeashLocation) _returnvalue_2 _returnvalue)))))

(fn ConstructionScript)
